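; Event runner that starts the LimeRock event from a given 0-1 percentage, a jump point used for dev debugging.
; It activates the LimeRock level instance data layer, then polls on a half-second delay until a BP_LimeRock_SequenceRunner actor exists and calls StartLimeRock on it.
; The sequence runner is deliberately hard-referenced because it uses little memory.
; Progress goes to the Fortnite shipping log, tagged with whether it runs on the server.
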(event Custom|LR_Event
  (params (percentage :type Float%20%28double-precision%29))
  ;[0.3126 0.5677 0.4905 0.5] Load the Event
  (Variables|Default|SetPercentage percentage) ; 0 - 1.0, what percentage to jump\nto when starting the event. Used\nfor dev debugging.
  (CallFunction|LRLog "Start the LimeRock event! (loading data layer)")
  (CallFunction|LRDLLoad) ; load the level instance into memory
  (Utilities|FlowControl|Delay 0.5 :!label _label)
  (CallFunction|LRLog "Looking for LimeRock Sequence Runner")
  (bind _outactors (Actor|GetAllActorsOfClass "/LimeRock/Gameplay/BP_LimeRock_SequenceRunner.BP_LimeRock_SequenceRunner_C")) ; it's ok to hard reference this because this\nblueprint takes up almost no memory
  (if (Utilities|Array|IsNotEmpty _outactors)
    (CallFunction|LRLog "Found the sequence runner! (starting event)")
    (Class|BPLimeRockSequenceRunner|StartLimeRock (Utilities|Array|Get(acopy) _outactors) (Variables|Default|GetPercentage))
    (else
      (goto _label))))

(event Custom|LR%20Log
  (params (string :type String))
  ;[0.6042 0.2988 0.4366 0.5] Shipping Log
  (Fortnite|FortShippingLog 0 (Utilities|String|Append "[BP_LimeRock_EventRunner] ( " (Utilities|String|ToString(Boolean) (Networking|IsServer)) " ) - " string)))

(event Custom|LR_DL_Load
  ;[0.2323 0.2396 0.4635 0.5] Data Layers
  (bind _returnvalue (WorldPartition|GetDataLayerManager))
  (DataLayers|SetDataLayerRuntimeState _returnvalue "/LimeRock/DataLayers/DL_LimeRock_LevelInstance.DL_LimeRock_LevelInstance" "Activated" false :!class /Script/Engine.DataLayerManager) ; preview actors
)

(fn ConstructionScript)
