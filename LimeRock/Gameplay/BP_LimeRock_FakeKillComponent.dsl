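; LimeRock component that plays cosmetic kill effects on a list of fake-kill relevancy victims.
; EventPlayKillEffects stores the victims, switches on the reason tag (Crash, Choke, Zap), then plays the elimination GameplayCue.
; The shared death FX function runs a local burst cue on each valid victim pawn where cosmetics can execute.
; Zap reuses that function with its own cue, while Crash and Choke loop over victims with empty bodies.
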(event EventBeginPlay)

(event EventTick
  (params (DeltaSeconds :type Float%20%28single-precision%29)))

(event FakeKillRelevancy|EventPlayKillEffects
  (params (FakeKillData_Victims :type Array%20of%20Fake%20Kill%20Relevancy%20Victim%20Structures) (FakeKillData_Reason :type Gameplay%20Tag%20Structure) (FakeKillData_PhaseNum :type Integer) (FakeKillData_WaveNum :type Integer) (RelevancyDelay :type Float%20%28single-precision%29))
  (Variables|Default|SetFake%20Kill%20Data%20Victims FakeKillData_Victims)
  (switch Utilities|FlowControl|Switch|SwitchonGameplayTag FakeKillData_Reason
    (:RipeHoneydew.Crash
      (CallFunction|PlayCrashFXOnVictims))
    (:RipeHoneydew.Choke
      (CallFunction|PlayChokeFXOnVictims))
    (:RipeHoneydew.Zap
      (CallFunction|PlayZapFXOnVictims)))
  (CallFunction|PlayDefaultDeathFXOnVictims (Variables|Default|GetElimination%20GC)))

(fn PlayChokeFXOnVictims
  (Networking|CanExecuteCosmeticEvents
    (:True
      (for _ (Variables|Default|GetFake%20Kill%20Data%20Victims)
        _))))

(fn PlayCrashFXOnVictims
  (Networking|CanExecuteCosmeticEvents
    (:True
      (for _ (Variables|Default|GetFake%20Kill%20Data%20Victims)
        _))))

(fn PlayDefaultDeathFXOnVictims
  (params (GameplayCueTag :type Gameplay%20Tag%20Structure))
  (Networking|CanExecuteCosmeticEvents
    (:True
      (for _array_element (Variables|Default|GetFake%20Kill%20Data%20Victims)
        (bind _array_element_playerstate (Utilities|Struct|BreakFakeKillRelevancyVictim _array_element))
        (Utilities|IsValid (PlayerState|GetPawn _array_element_playerstate)
          (:Is%20Valid
            (GameplayCue|ExecuteGameplayCue%28Burst%29Local (PlayerState|GetPawn _array_element_playerstate) (Variables|GetGameplayCueTag) (Ability|GameplayCue|MakeGameplayCueParameters 0.0))))))))

(fn PlayZapFXOnVictims
  (Networking|CanExecuteCosmeticEvents
    (:True
      (CallFunction|PlayDefaultDeathFXOnVictims (Variables|Default|GetZapGC)))))
