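; LimeRock knockback volume actor that launches Fort pawns, enabling either its Box or Sphere collision at BeginPlay.
; A velocity event throws a non-falling pawn against player 0's velocity, with added height.
; A pulse event sphere-traces around the actor or a custom point, launching pawns away from the actor using strafe or after-hover tuning, with optional local-viewer audio.
; Hit and overlap handlers are empty, so neither knockback fires from collision here.
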(event Custom|KnockbackPlayerFromVelocity
  (params (Player :type Fort%20Pawn%20Object%20Reference))
  ; knockback
  (if (not (AI|Components|NavMovement|IsFalling (Variables|/Script/Engine.Character|GetCharacterMovement Player)))
    (Pawn|Character|LaunchCharacterJump Player (Math|Vector|MakeVector (.x (* (* (Math|Vector|Normalize (Transformation|GetVelocity (Game|GetPlayerPawn 0))) -1.0) (Variables|Default|GetKnockbackDistance))) (.y (* (* (Math|Vector|Normalize (Transformation|GetVelocity (Game|GetPlayerPawn 0))) -1.0) (Variables|Default|GetKnockbackDistance))) (+ (.z (* (* (Math|Vector|Normalize (Transformation|GetVelocity (Game|GetPlayerPawn 0))) -1.0) (Variables|Default|GetKnockbackDistance))) (Variables|Default|GetKnockbackHeight))) true true true true)
    (Audio|PlaySoundatLocation 0 (Transformation|GetActorLocation Player))))

(event EventBeginPlay
  ; begin play
  (Utilities|FlowControl|Sequence
    (:then_0
      (CallFunction|SetReferences))
    (:then_1
      (switch Utilities|FlowControl|Switch|SwitchonECollisionShapeMode (Variables|Default|GetCollisionShape)
        (:Box
          (CallFunction|SetKnockbackCollision true "Box")
          (CallFunction|SetKnockbackCollision false "Sphere"))
        (:Sphere
          (CallFunction|SetKnockbackCollision false "Box")
          (CallFunction|SetKnockbackCollision true "Sphere"))))))

(event Custom|KnockbackPlayerPulseFromCenter
  (params (UseCustomLocation :type Boolean) (CustomLocation :type Vector) (Radius :type Float%20%28single-precision%29 :default 1e+04) (PreEvent :type Boolean))
  ; pulse
  (bind (_outhits _returnvalue) (Collision|MultiSphereTraceForObjects (select UseCustomLocation CustomLocation (Transformation|GetActorLocation)) (+ (select UseCustomLocation CustomLocation (Transformation|GetActorLocation)) 1) (select UseCustomLocation (ScalableFloat|GetValueatLevel (Variables|Default|GetStrafeKnockbackPulseRadius)) 1e+04) (Utilities|Array|MakeArray "ObjectTypeQuery3")))
  (if _returnvalue
    (for _array_element _outhits
      (bind _asfort_pawn (Utilities|Casting|CastToFortPawn (.HitActor (Collision|BreakHitResult _array_element))))
      (Utilities|IsValid _asfort_pawn
        (:Is%20Valid
          (Pawn|Character|LaunchCharacterJump _asfort_pawn (Math|Vector|MakeVector (.x (* (* (Math|Vector|Normalize (- (Transformation|GetActorLocation) (Transformation|GetActorLocation _asfort_pawn))) -1.0) (ScalableFloat|GetValueatLevel (select UseCustomLocation (Variables|Default|GetStrafeKnockbackPulseDistance) (Variables|Default|GetAfterHoverKnockbackPulseDistance))))) (.y (* (* (Math|Vector|Normalize (- (Transformation|GetActorLocation) (Transformation|GetActorLocation _asfort_pawn))) -1.0) (ScalableFloat|GetValueatLevel (select UseCustomLocation (Variables|Default|GetStrafeKnockbackPulseDistance) (Variables|Default|GetAfterHoverKnockbackPulseDistance))))) (+ (.z (* (* (Math|Vector|Normalize (- (Transformation|GetActorLocation) (Transformation|GetActorLocation _asfort_pawn))) -1.0) (ScalableFloat|GetValueatLevel (select UseCustomLocation (Variables|Default|GetStrafeKnockbackPulseDistance) (Variables|Default|GetAfterHoverKnockbackPulseDistance))))) (ScalableFloat|GetValueatLevel (select UseCustomLocation (Variables|Default|GetStrafeKnockbackPulseHeight) (Variables|Default|GetAfterHoverKnockbackPulseHeight))))) true true true true)
          (if PreEvent
            (if (Pawn|IsLocallyViewed _asfort_pawn)
              (Variables|Default|GetKnockBackAudio
                (:then)
                (:else
                  (bind _returnvalue_2 (Audio|SpawnSoundAttached (Variables|Default|GetKnockBackSound) (Variables|/Script/Engine.Actor|GetRootComponent _asfort_pawn)))
                  (Variables|Default|SetKnockBackAudio _returnvalue_2))))))))))

(event Custom|SetKnockbackCollision
  (params (Enable :type Boolean) (CollisionShape :type ECollisionShapeMode%20Enum :default "Mesh"))
  (bind _returnvalue (select Enable "QueryAndPhysics" "NoCollision"))
  ; set collision
  (switch Utilities|FlowControl|Switch|SwitchonECollisionShapeMode CollisionShape
    (:Box
      (Collision|SetCollisionEnabled (Variables|Default|GetBox) _returnvalue))
    (:Sphere
      (Collision|SetCollisionEnabled (Variables|Default|GetSphere) _returnvalue))))

(event OnComponentHit(Sphere)
  (params (HitComponent :type Primitive%20Component%20Object%20Reference) (OtherActor :type Actor%20Object%20Reference) (OtherComp :type Primitive%20Component%20Object%20Reference) (NormalImpulse :type Vector :default "0, 0, 0") (Hit :type Hit%20Result%20Structure%20%28by%20ref%29)))

(event OnComponentHit(Box)
  (params (HitComponent :type Primitive%20Component%20Object%20Reference) (OtherActor :type Actor%20Object%20Reference) (OtherComp :type Primitive%20Component%20Object%20Reference) (NormalImpulse :type Vector :default "0, 0, 0") (Hit :type Hit%20Result%20Structure%20%28by%20ref%29)))

(event OnComponentBeginOverlap(Sphere)
  (params (OverlappedComponent :type Primitive%20Component%20Object%20Reference) (OtherActor :type Actor%20Object%20Reference) (OtherComp :type Primitive%20Component%20Object%20Reference) (OtherBodyIndex :type Integer) (bFromSweep :type Boolean) (SweepResult :type Hit%20Result%20Structure%20%28by%20ref%29)))

(event OnComponentBeginOverlap(Box)
  (params (OverlappedComponent :type Primitive%20Component%20Object%20Reference) (OtherActor :type Actor%20Object%20Reference) (OtherComp :type Primitive%20Component%20Object%20Reference) (OtherBodyIndex :type Integer) (bFromSweep :type Boolean) (SweepResult :type Hit%20Result%20Structure%20%28by%20ref%29)))

(fn SetReferences
  (bind _asfort_player_pawn (Utilities|Casting|CastToFortPlayerPawn (Game|GetPlayerPawn 0)))
  (Variables|Default|SetPlayer _asfort_player_pawn))

(fn ConstructionScript)
