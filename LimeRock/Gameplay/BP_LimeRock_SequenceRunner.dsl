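; LimeRock SequenceRunner Blueprint with helpers for a sequenced calendar event and its video playback.
; It asks the squad finder actor to generate a pawn list on the server, logging if none exists.
; Video state replication forwards video events on non-server instances, and a scalable float reports whether the new video flow is on.
; Remaining pre-event time is the replicated sequence player's end frame minus current frame, divided by a hardcoded 60.
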(fn GeneratePawnList
  (bind _returnvalue (Actor|GetActorOfClass "/LimeRock/Gameplay/BP_Sequencer_SquadFinder.BP_Sequencer_SquadFinder_C"))
  (Utilities|IsValid _returnvalue
    (:Is%20Valid
      (Class|BPSequencerSquadFinder|%28Server%29GeneratePawnList _returnvalue))
    (:Is%20Not%20Valid
      (Fortnite|FortShippingLog 0 "Couldn't find the player list creator blueprint" true))) ; found the player finder blueprint?
)

(fn LogIt
  (params (String :type String) (bLogAsWarning :type Boolean))
  (Fortnite|FortShippingLog 0 (Utilities|String|Append "[LimeRock SequenceRunner] " String) bLogAsWarning))

(fn LogVideo
  (params (String :type String) (bLogAsWarning :type Boolean))
  (CallFunction|LogIt (Utilities|String|Append " (Limerock Video - B_FortVideoWidgetBase_LimeRock) " String) bLogAsWarning))

(fn NewVideoFlow?
  (return (ScalableFloat|GetValueasBool (Variables|Video|GetUse%20New%20Video%20Flow))))

(fn OnRep_VideoState
  (if (not (Networking|IsServer))
    (CallFunction|SendVideoEventOnClient (Variables|Video|GetVideo%20State))))

(fn RemainingPreEventTime
  (bind _returnvalue_frame (Utilities|TimeManagement|BreakQualifiedFrameTime (Sequencer|Player|GetCurrentTime (Variables|/Script/LevelSequence.LevelSequenceActor|GetSequencePlayer (Variables|CalendarEventSequencePlayer|GetReplicated%20Sequence%20Player%20%28OFPA%29%20-%20K&K)))))
  (bind _returnvalue_frame_value (Utilities|Struct|BreakFrameNumber _returnvalue_frame))
  (bind _returnvalue_frame_2 (Utilities|TimeManagement|BreakQualifiedFrameTime (Sequencer|Player|GetEndTime (Variables|/Script/LevelSequence.LevelSequenceActor|GetSequencePlayer (Variables|CalendarEventSequencePlayer|GetReplicated%20Sequence%20Player%20%28OFPA%29%20-%20K&K)) :!class /Script/MovieScene.MovieSceneSequencePlayer)))
  (bind _returnvalue_frame_value_2 (Utilities|Struct|BreakFrameNumber _returnvalue_frame_2))
  (return (/ (- _returnvalue_frame_value_2 _returnvalue_frame_value) 60)))

(fn ConstructionScript
  (|CallParentFunction))
