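; UFO camera shake actor that drives proximity-based camera shake and controller rumble for UFO flybys and abductions.
; It listens for stateful UFO events on the global event router, with reset logic clearing timer and radii.
; Flyby runs a looping timer that scales attenuation radii by the attach parent's speed.
; Abduction uses fixed tunable radii, sharing shake radii with rumble; scalable floats can disable either effect.
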(event EventBeginPlay
  (CallFunction|SetReferences)
  (CallFunction|BindCameraShake))

(event Collision|EventActorBeginOverlap
  (params (OtherActor :type Actor%20Object%20Reference)))

(event Custom|SetAttenuation
  (bind _returnvalue (Actor|GetAttachParentActor))
  (bind _returnvalue_2 (Transformation|GetVelocity _returnvalue))
  (bind _returnvalue_3 (Math|Vector|VectorLength _returnvalue_2))
  (bind _returnvalue_4 (- _returnvalue_3 1.5e+03))
  (bind _returnvalue_5 (/ _returnvalue_4 3.5e+03))
  (bind _camerashakesource (Variables|Default|GetCameraShakeSource))
  ; set camera shake attenuation based on velocity of parent actor
  (Variables|Default|SetCurrentVelocity _returnvalue_3)
  (Variables|/Script/Engine.CameraShakeSourceComponent|SetInnerAttenuationRadius (Math|Float|Lerp 1.0 (ScalableFloat|GetValueatLevel (Variables|Default|GetUFO_Flyby_ShakeInnerRadiusMax)) (Math|Float|Clamp(Float) _returnvalue_5)) _camerashakesource)
  (Variables|/Script/Engine.CameraShakeSourceComponent|SetOuterAttenuationRadius (Math|Float|Lerp 1.0 (ScalableFloat|GetValueatLevel (Variables|Default|GetUFO_Flyby_ShakeOuterRadiusMax)) (Math|Float|Clamp(Float) _returnvalue_5)) _camerashakesource)
  ; disable flyby rumble?
  (if (not (>= (ScalableFloat|GetValueatLevel (Variables|Default|GetUFO_FlyBy_DisableRumble)) 1.0))
    ; FF rumble attenuation
    (ForceFeedback|AdjustAttenuation (Variables|Default|GetForceFeedback) (Utilities|Struct|MakeForceFeedbackAttenuationSettings "Logarithmic" "Sphere" "Continues" -6e+01 (Math|Vector|MakeVector (Math|Float|Lerp 1.0 (ScalableFloat|GetValueatLevel (Variables|Default|GetUFO_Flyby_RumbleInnerRadiusMax)) (Math|Float|Clamp(Float) _returnvalue_5))) 0.0 (Math|Float|Lerp 1.0 (ScalableFloat|GetValueatLevel (Variables|Default|GetUFO_Flyby_RumbleOuterRadiusMax)) (Math|Float|Clamp(Float) _returnvalue_5)) 0.0 0.0 "(EditorCurveData=(Keys=,DefaultValue=340282346638528859811704183484516925440.000000,PreInfinityExtrap=RCCE_Constant,PostInfinityExtrap=RCCE_Constant),ExternalCurve=None)"))))

(event Custom|BindCameraShake
  (bind _camerashakesource (Variables|Default|GetCameraShakeSource))
  (bind _camerashakesource_2 (Variables|Default|GetCameraShakeSource))
  ; bind camera shake start/stop
  (bind (_asynctaskproxy _payload) (GameplayEventRouter|StartListeningToStatefulEvent (GameplayEventRouter|GetGlobalEventRouter) "/LimeRock/Gameplay/EventRouter/LimeRock_UFO_EventRouter.LimeRock_UFO_EventRouter" (GameplayEventRouter|GetGlobalEventRouter))
    (:OnEventReceived
      (switch Utilities|FlowControl|Switch|SwitchonE_LimeRock_UFO_EventRouter (Utilities|Struct|BreakLimeRockUFOEventRouter _payload)
        (:NewEnumerator0
          ; disable flyby shake?
          (if (>= (ScalableFloat|GetValueatLevel (Variables|Default|GetUFO_FlyBy_DisableShake)) 1.0)
            (Components|Activation|Deactivate (Variables|Default|GetCameraShakeSource))
            ; disable flyby rumble?
            (if (not (>= (ScalableFloat|GetValueatLevel (Variables|Default|GetUFO_FlyBy_DisableRumble)) 1.0))
              ; FF rumble attenuation
              (ForceFeedback|AdjustAttenuation (Variables|Default|GetForceFeedback) (Utilities|Struct|MakeForceFeedbackAttenuationSettings "Logarithmic" "Sphere" "Continues" -6e+01 (Math|Vector|MakeVector (Math|Float|Lerp 1.0 (ScalableFloat|GetValueatLevel (Variables|Default|GetUFO_Flyby_RumbleInnerRadiusMax)) (Math|Float|Clamp(Float) (/ (- (Math|Vector|VectorLength (Transformation|GetVelocity (Actor|GetAttachParentActor))) 1.5e+03) 3.5e+03)))) 0.0 (Math|Float|Lerp 1.0 (ScalableFloat|GetValueatLevel (Variables|Default|GetUFO_Flyby_RumbleOuterRadiusMax)) (Math|Float|Clamp(Float) (/ (- (Math|Vector|VectorLength (Transformation|GetVelocity (Actor|GetAttachParentActor))) 1.5e+03) 3.5e+03))) 0.0 0.0 "(EditorCurveData=(Keys=,DefaultValue=340282346638528859811704183484516925440.000000,PreInfinityExtrap=RCCE_Constant,PostInfinityExtrap=RCCE_Constant),ExternalCurve=None)")))
            (else
              ; set camera shake attenuation based on velocity of parent actor
              (Components|Activation|Activate (Variables|Default|GetCameraShakeSource) false)
              (bind _returnvalue (Utilities|Time|SetTimerbyEvent (delegate Custom|SetAttenuation) (ScalableFloat|GetValueatLevel (Variables|Default|GetUFO_Flyby_DetectionFrequency)) true true))
              (Variables|Default|SetSetAttenuationTimer _returnvalue))))
        (:NewEnumerator1
          _)
        (:NewEnumerator2
          ; disable abduct shake?
          (if (>= (ScalableFloat|GetValueatLevel (Variables|Default|GetUFO_Abduct_DisableShake)) 1.0)
            (Components|Activation|Deactivate (Variables|Default|GetCameraShakeSource))
            (else
              ; abduction camera shake
              (Components|Activation|Activate (Variables|Default|GetCameraShakeSource) false)
              (Variables|/Script/Engine.CameraShakeSourceComponent|SetInnerAttenuationRadius (ScalableFloat|GetValueatLevel (Variables|Default|GetUFO_Abduct_ShakeInnerRadius)) _camerashakesource_2)
              (Variables|/Script/Engine.CameraShakeSourceComponent|SetOuterAttenuationRadius (ScalableFloat|GetValueatLevel (Variables|Default|GetUFO_Abduct_ShakeOuterRadius)) _camerashakesource_2)))
          ; disable flyby rumble?
          (if (not (>= (ScalableFloat|GetValueatLevel (Variables|Default|GetUFO_Abduct_DisableRumble)) 1.0))
            ; abduction rumble
            (ForceFeedback|AdjustAttenuation (Variables|Default|GetForceFeedback) (Utilities|Struct|MakeForceFeedbackAttenuationSettings "Logarithmic" "Sphere" "Continues" -6e+01 (Math|Vector|MakeVector (ScalableFloat|GetValueatLevel (Variables|Default|GetUFO_Abduct_ShakeInnerRadius))) 0.0 (ScalableFloat|GetValueatLevel (Variables|Default|GetUFO_Abduct_ShakeOuterRadius)) 0.0 0.0 "(EditorCurveData=(Keys=,DefaultValue=340282346638528859811704183484516925440.000000,PreInfinityExtrap=RCCE_Constant,PostInfinityExtrap=RCCE_Constant),ExternalCurve=None)"))))
        (:NewEnumerator3
          _))
      (Components|Activation|Deactivate (Variables|Default|GetCameraShakeSource))
      ; stop timer
      (Utilities|Time|ClearandInvalidateTimerbyHandle (Variables|Default|GetSetAttenuationTimer))
      ; reset camera shake attenuation
      (Variables|/Script/Engine.CameraShakeSourceComponent|SetInnerAttenuationRadius 1.0 _camerashakesource)
      (Variables|/Script/Engine.CameraShakeSourceComponent|SetOuterAttenuationRadius 1.0 _camerashakesource)
      ; reset rumble attenuation
      (ForceFeedback|AdjustAttenuation (Variables|Default|GetForceFeedback) (Utilities|Struct|MakeForceFeedbackAttenuationSettings "Logarithmic" "Sphere" "Continues" -6e+01 (1.0 0.0 0.0) 0.0 1.0 0.0 0.0 "(EditorCurveData=(Keys=,DefaultValue=340282346638528859811704183484516925440.000000,PreInfinityExtrap=RCCE_Constant,PostInfinityExtrap=RCCE_Constant),ExternalCurve=None)")))))

(fn SetReferences
  (bind _asfort_player_pawn (Utilities|Casting|CastToFortPlayerPawn (Game|GetPlayerPawn 0)))
  (Variables|Default|SetFortPlayerPawn _asfort_player_pawn))

(fn ConstructionScript)
