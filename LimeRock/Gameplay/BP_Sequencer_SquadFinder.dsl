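; BP_Sequencer_SquadFinder supplies LimeRock sequences with a replicated list of player pawns.
; The server-only GeneratePawnList stores every Fortnite player pawn without squad filtering, and GetPawnList returns that list.
; It logs a "total squadmates" figure using the array's last index, one below the count, through a helper that tags server or client.
; BeginPlay is empty, so nothing here triggers generation.
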(event EventBeginPlay)

(event Custom|Log%20It
  (params (string :type String))
  (Fortnite|FortShippingLog 0 (Utilities|String|Append (Utilities|String|SelectString "server" "client" (Networking|IsServer)) "[LimeRock] Sequence Squad Finder - " string)))

(fn %28server%29GeneratePawnList
  (if (Networking|IsServer)
    (bind _outfortplayerpawns (Fortnite|GetFortPlayerPawns))
    (bind _output_get (Variables|Default|SetReplicated%20Player%20Pawn%20List _outfortplayerpawns))
    (CallFunction|LogIt (Utilities|String|Append " total squadmates: " (Utilities|String|ToString(Integer) (Utilities|Array|LastIndex _output_get))))))

(fn GetPawnList
  (return (Variables|Default|GetReplicated%20Player%20Pawn%20List)))

(fn ConstructionScript)
