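; Blueprint function library for LimeRock gameplay that hides or shows HUD elements for the local player.
; HideHUDForLocalPlayer takes a gameplay tag container, a hide flag and a log context string.
; When cosmetic events can execute and the first local Fort player controller is valid, it sets those elements' visibility.
; It passes the reason "Playlist" from GetHUDHideReason, which must match the playlist tag for playlist-driven hiding.
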(fn GetHUDHideReason
  ;[0.15 0.15 0.15 0.5 12] This reason corresponds to ReasonName in FortPlayerController::SetHUDElementVisibilityInternal. Needs to match the playlist tag if being hidden by the playlist
  (return "Playlist"))

(fn HideHUDForLocalPlayer
  (params (HUDElementsToHide :type Gameplay%20Tag%20Container%20Structure) (Hide :type Boolean) (LogContext :type String))
  (bind _returnvalue (Fortnite|GetFirstLocalFortPlayerController true))
  (Networking|CanExecuteCosmeticEvents
    (:True
      (Utilities|IsValid _returnvalue
        (:Is%20Valid
          (HUD|SetHUDElementVisibilitywithContext _returnvalue (Variables|GetHUDElementsToHide) (Variables|GetHide) (CallFunction|GetHUDHideReason 0) (Variables|GetLogContext)))))))
